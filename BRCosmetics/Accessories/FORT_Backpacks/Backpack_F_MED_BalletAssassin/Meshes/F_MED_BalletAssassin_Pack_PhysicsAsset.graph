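# Physics asset for the BalletAssassin backpack cosmetic, defining capsule bodies and joint limits for its dangling ribbons and strap.
# Two two-link ribbon chains (top and bottom) hang from their base bones, and a six-link strap chain hangs from base_strap, which has a large capsule.
# The pack body combines three angled capsules and has no constraints.
# Strap swing is mostly single-axis with little twist, loosening down the chain until a stiffer final link.
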PhysicsAsset /BRCosmetics/Accessories/FORT_Backpacks/Backpack_F_MED_BalletAssassin/Meshes/F_MED_BalletAssassin_Pack_PhysicsAsset
bodies {
 base_ribbon_b [{shapeType=Capsule, length=0.505, radius=0.505, shapeName=base_ribbon_b_capsule, rotation=<rot 0 -90 0>}]
 base_ribbon_t [{shapeType=Capsule, length=0.505, radius=0.505, shapeName=base_ribbon_t_capsule, rotation=<rot 0 -90 0>}]
 base_strap [{shapeType=Capsule, length=18, radius=15.505, shapeName=base_strap_capsule, center=<18.8661 -17.5787 0>, rotation=<rot 90 180 -175.1457>}]
 dyn_ribbon_01_b [{shapeType=Capsule, length=0.9136, radius=0.75, shapeName=dyn_ribbon_01_b_capsule, center=<1.1287 -0.0783 0>, rotation=<rot 0 -93.9705 90>}]
 dyn_ribbon_01_t [{shapeType=Capsule, length=0.9728, radius=0.75, shapeName=dyn_ribbon_01_t_capsule, center=<0.9624 0 0>, rotation=<rot 0 90 90>}]
 dyn_ribbon_02_b [{shapeType=Capsule, length=0.0026, radius=2.0782, shapeName=dyn_ribbon_02_b_capsule, center=<1.0344 0 0>, rotation=<rot 0 -90 90>}]
 dyn_ribbon_02_t [{shapeType=Capsule, length=0.2491, radius=1.7062, shapeName=dyn_ribbon_02_t_capsule, center=<1.1983 0 0>, rotation=<rot 0 -90 90>}]
 dyn_strap_01 [{shapeType=Capsule, length=1.75, radius=0.75, center=<2.163 0 0>, rotation=<rot 90 0 0>}]
 dyn_strap_02 [{shapeType=Capsule, length=2.5, radius=0.75, center=<2.1275 0 0>, rotation=<rot 90 0 0>}]
 dyn_strap_03 [{shapeType=Capsule, length=1.65, radius=1.25, center=<2.25 0 0>, rotation=<rot 90 0 0>}]
 dyn_strap_04 [{shapeType=Capsule, length=1.75, radius=1.5, center=<2.75 0 0>, rotation=<rot 90 0 0>}]
 dyn_strap_05 [{shapeType=Capsule, length=2, radius=1.5, center=<2.75 0 0>, rotation=<rot 90 0 0>}]
 dyn_strap_06 [{shapeType=Capsule, length=1, radius=1.5, center=<2.5 0 0>, rotation=<rot 90 0 0>}]
 pack [
  {
   shapeType=Capsule
   length=36.8268
   radius=1.9898
   shapeName=pack_capsule
   center=<-4.7494 -3.2621 -9.9108>
   rotation=<rot 0.8509 -90 54.2229>
  }
  {
   shapeType=Capsule
   length=26
   radius=2.4898
   shapeName=pack_capsule
   center=<-8.8389 -2.9561 -5.5045>
   rotation=<rot -29.2383 0.0018 -0.0037>
  }
  {shapeType=Capsule, length=17, radius=4, shapeName=pack_capsule, center=<-2.7904 -4.7286 -3.3517>, rotation=<rot -47.3563 0.0006 2.0185>}
 ]
}
constraints [
 {bone1Name=dyn_ribbon_01_b, bone2Name=base_ribbon_b, swing1LimitDegrees=20, swing2LimitDegrees=12, twistLimitDegrees=2.5}
 {bone1Name=dyn_ribbon_02_b, bone2Name=dyn_ribbon_01_b, swing1LimitDegrees=15, swing2LimitDegrees=20, twistLimitDegrees=2.5}
 {bone1Name=dyn_ribbon_01_t, bone2Name=base_ribbon_t, swing1LimitDegrees=20, swing2LimitDegrees=20, twistLimitDegrees=2.5}
 {bone1Name=dyn_ribbon_02_t, bone2Name=dyn_ribbon_01_t, swing1LimitDegrees=40, swing2LimitDegrees=25, twistLimitDegrees=2.5}
 {bone1Name=dyn_strap_01, bone2Name=base_strap, swing1LimitDegrees=30, swing2LimitDegrees=4, twistLimitDegrees=5}
 {bone1Name=dyn_strap_02, bone2Name=dyn_strap_01, swing1LimitDegrees=30, swing2LimitDegrees=6, twistLimitDegrees=5}
 {bone1Name=dyn_strap_03, bone2Name=dyn_strap_02, swing1LimitDegrees=35, swing2LimitDegrees=8, twistLimitDegrees=5}
 {bone1Name=dyn_strap_04, bone2Name=dyn_strap_03, swing1LimitDegrees=40, swing2LimitDegrees=10, twistLimitDegrees=5}
 {bone1Name=dyn_strap_05, bone2Name=dyn_strap_04, swing1LimitDegrees=45, swing2LimitDegrees=10, twistLimitDegrees=5}
 {bone1Name=dyn_strap_06, bone2Name=dyn_strap_05, swing1LimitDegrees=10, swing2LimitDegrees=10, twistLimitDegrees=5}
]
